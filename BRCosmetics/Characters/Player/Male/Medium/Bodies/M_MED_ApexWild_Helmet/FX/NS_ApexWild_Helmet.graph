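# Niagara effect for the ApexWild helmet cosmetic, adding red glows gated by User.IsEnabled.
# Arm_Glow holds one looping particle with a mesh, sine-pulsed colored light and component renderer fed spotlight-style angle and rotation attributes.
# Arm_Glow001 continuously spawns small randomized meshes that fade and swell over life, with brighter color at night.
# Head_Glow pins one pulsing sprite-and-mesh glow to User.SkeletalMesh, phase-offset from the arm pulse.
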NiagaraSystem /BRCosmetics/Characters/Player/Male/Medium/Bodies/M_MED_ApexWild_Helmet/FX/NS_ApexWild_Helmet
dependencies {usedDataInterfaces [&NiagaraDataInterfaceCurve], usedDynamicInputs [], usedModules [&SystemState]}
emitters {
 Arm_Glow {
  emitterUpdateScript [
   &EmitterState()
   &SpawnBurst_Instantaneous(
    Age:NiagaraFloat=$Emitter.Age
    Spawn Count=&Multiply_Int(A=1, B=&MakeIntFromFloat(Float=$User.IsEnabled))
    Spawn Time=0.01
    Use Spawn Probability=false
    UseLoopCountLimit=false
   )
  ]
  particleSpawnScript [
   &InitializeParticle(
    Color=<rgba 1 0.0282 0.02 500>
    Lifetime=5
    Sprite Size Mode=::NewEnumerator3
    Uniform Sprite Size=2
    UsePositionOffset=false
    Write Material Random=true
   )
   &InitialMeshOrientation(Mesh Axis To Orient=<0 0 1>, Mesh Orientation Mode=::NewEnumerator2, Rotation=<0 0.5 0>, Use Rotation Vector=true)
   &InitialMeshOrientation(Mesh Orientation Mode=::NewEnumerator7, Rotation=<0 0 0.5>, Use Rotation Vector=true) {module=InitialMeshOrientation001}
  ]
  particleUpdateScript [
   &ParticleState(
    Kill Particles When Lifetime Has Elapsed=false
    Let Infinitely Lived Particles Die When Emitter Deactivates=true
    Loop Particles Lifetime=true
   )
   &FN_Gen5_Light(Attenutation=20, Color:LinearColor=<rgba 1 0.15 0.15 1>, Exponent=5, Intensity=100, Position Offset:Vector3f=<0 -0.5 -10>)
   &FN_Light_Attributes(
    Color=<rgba 1 0.02 0.02 1>
    Create Color Attribute=true
    Create Exponent Attribute=true
    Create Position Attribute=true
    Create Radius Attribute=true
    Exponent=5
    Position=<-1 -0.5 -10>
    Radius=20
    Scale Color=&Lerp_Float(A=2, B=3, Alpha=&Sine(Normalized Angle=$Emitter.Age, Period=3, Scale=0.5, Bias=0.5))
   )
   &ScaleColor(
    Scale RGB=&VectorFromFloat(Value=&Lerp_Float(A=5, B=5, Alpha=&Sine(Normalized Angle=$Emitter.Age, Period=3, Scale=0.5, Bias=0.5)))
    ScaleA=true
    ScaleRGB=true
   )
   /SetVariables(Particles.FN LightPassthrough:Vector3f=$Particles.FN_Light_Attributes.Position) {isSetParameters=true}
   /SetVariables(Particles.Gen5LightRotation:Vector3f=<-60 0 0>) {isSetParameters=true}
   /SetVariables(Particles.LightAngleInner=65) {isSetParameters=true}
   /SetVariables(Particles.LightAngleOuter=80) {isSetParameters=true}
  ]
  rendererClasses [
   &NiagaraComponentRendererProperties
   &NiagaraLightRendererProperties
   &NiagaraMeshRendererProperties
  ]
  renderers [
   {rendererClass=&NiagaraComponentRendererProperties, rendererIndex=0}
   {rendererClass=&NiagaraLightRendererProperties, rendererIndex=1}
   {rendererClass=&NiagaraMeshRendererProperties, rendererIndex=2}
  ]
 }
 Arm_Glow001 {
  emitterUpdateScript [
   &EmitterState()
   &SpawnRate(SpawnRate=&Multiply_Float(A=2, B=$User.IsEnabled), Use Spawn Probability=false)
   &SpawnRate(SpawnRate=90, Use Spawn Probability=false) {enabled=false, module=SpawnRate001}
  ]
  particleSpawnScript [
   &InitializeParticle(
    Color=&FN_Select_Color_from_Time_Of_Day(Morning=<rgba 1 0.0282 0.02 1>, Day=<rgba 1 0.0282 0.02 1>, Evening=<rgba 1 0.0282 0.02 1>, Night=<rgba 5 0.1408 0.1 1>)
    Lifetime Max=3
    Lifetime Min=6
    Lifetime Mode=::NewEnumerator1
    Mesh Scale=&Multiply_VectorByFloat(Vector=&RandomRangeVector(Minimum=<0.05 0.05 0.4>, Maximum=<0.1 0.1 0.25>), Float=1)
    Mesh Scale Mode=::NewEnumerator4
    Position Offset=<-7 0.75 -9>
    Sprite Size Mode=::NewEnumerator3
    Uniform Sprite Size=2
    UsePositionOffset=true
    Write Material Random=true
   )
   &InitialMeshOrientation(
    Mesh Orientation Mode=::NewEnumerator7
    Rotation=&Add_Vector(
     A=&Multiply_VectorByFloat(Vector=&RandomRangeVector(Minimum=<-3 -1 -3>, Maximum=<3 1 1>), Float=0.025)
     B=<0 0.32 0>
    )
    Use Rotation Vector=true
   ) {enabled=false}
   &ShapeLocation(
    MaskSpawnGroup=false
    Non Uniform Scale=<0 2 0>
    Offset=<1 0 1>
    Offset Mode=::NewEnumerator0
    Override Seed=false
    Sphere Radius=0.6
    UseOwnerScale=true
   )
   &InitialMeshOrientation(Mesh Orientation Mode=::NewEnumerator7, Rotation=<0 0.27 0>, Use Rotation Vector=true) {module=InitialMeshOrientation001}
   &InitialMeshOrientation(
    Mesh Orientation Mode=::NewEnumerator7
    Rotation=&RandomRangeVector(Minimum=<0 0 0.05>, Maximum=<0 0 -0.05>)
    Use Rotation Vector=true
   ) {module=InitialMeshOrientation002}
  ]
  particleUpdateScript [
   &ParticleState()
   &ScaleColor(
    Scale Alpha=&FloatFromCurve(FloatCurve=DataInterface() {propertyValues {Curve {keys [
         {interpMode=RCIM_Cubic, tangentMode=RCTM_User, time=0, value=0}
         {interpMode=RCIM_Cubic, tangentMode=RCTM_User, time=0.5, value=1}
         {tangentMode=RCTM_User, time=1, value=0}
        ]}}})
    Scale RGB=&VectorFromFloat(Value=&FloatFromCurve(
      FloatCurve=DataInterface() {propertyValues {Curve {keys [
          {interpMode=RCIM_Cubic, tangentMode=RCTM_User, time=0, value=0}
          {interpMode=RCIM_Cubic, tangentMode=RCTM_User, time=0.5, value=1}
          {tangentMode=RCTM_User, time=1, value=0}
         ]}}}
      Scale Curve=&RandomRangeFloat(Minimum=1, Maximum=5)
     ))
    ScaleA=true
    ScaleRGB=true
   )
   &ScaleMeshSize(Scale Factor=&MakeVector(
     X=&FloatFromCurve(FloatCurve=DataInterface() {propertyValues {Curve {keys [
          {interpMode=RCIM_Cubic, tangentMode=RCTM_User, time=0, value=0.5}
          {interpMode=RCIM_Cubic, tangentMode=RCTM_User, time=0.5, value=1}
          {tangentMode=RCTM_User, time=1, value=0.5}
         ]}}})
     Y=&FloatFromCurve(FloatCurve=DataInterface() {propertyValues {Curve {keys [
          {interpMode=RCIM_Cubic, tangentMode=RCTM_User, time=0, value=0.5}
          {interpMode=RCIM_Cubic, tangentMode=RCTM_User, time=0.5, value=1}
          {tangentMode=RCTM_User, time=1, value=0.5}
         ]}}})
     Z=&FloatFromCurve(FloatCurve=DataInterface() {propertyValues {Curve {keys [
          {interpMode=RCIM_Cubic, tangentMode=RCTM_User, time=0, value=0.75}
          {interpMode=RCIM_Cubic, tangentMode=RCTM_User, time=0.5, value=1}
          {tangentMode=RCTM_User, time=1, value=0.75}
         ]}}})
    ))
  ]
  rendererClasses [&NiagaraMeshRendererProperties]
  renderers [{rendererClass=&NiagaraMeshRendererProperties, rendererIndex=0}]
 }
 Head_Glow {
  emitterUpdateScript [
   &EmitterState()
   &SpawnBurst_Instantaneous(
    Age:NiagaraFloat=$Emitter.Age
    Spawn Count=&Multiply_Int(A=1, B=&MakeIntFromFloat(Float=$User.IsEnabled))
    Use Spawn Probability=false
    UseLoopCountLimit=false
   ) {module=SpawnBurst_Instantaneous001}
  ]
  particleSpawnScript [
   &InitializeParticle(
    Color=<rgba 0.4564 0.0046 0.0204 1>
    Lifetime=5
    Position Mode=::NewEnumerator0
    Sprite Size Mode=::NewEnumerator3
    Uniform Sprite Size=50
    Write Material Random=true
   )
   &InitialMeshOrientation(Mesh Orientation Mode=::NewEnumerator7, Rotation=<0 0.25 0>, Use Rotation Vector=true)
  ]
  particleUpdateScript [
   &ParticleState(
    Kill Particles When Lifetime Has Elapsed=false
    Let Infinitely Lived Particles Die When Emitter Deactivates=true
    Loop Particles Lifetime=true
   )
   &ScaleColor(
    Scale RGB=&VectorFromFloat(Value=&Lerp_Float(
      A=10
      B=15
      Alpha=&Sine(Normalized Angle=&Add_Float(A=$Emitter.Age, B=1), Period=3, Scale=0.5, Bias=0.5)
     ))
    ScaleA=true
    ScaleRGB=true
   )
   &SkeletalMeshLocation(Bone Sampling Mode=::NewEnumerator4, MaskSpawnGroup=false, Skeletal Mesh=$User.SkeletalMesh)
  ]
  rendererClasses [&NiagaraSpriteRendererProperties, &NiagaraMeshRendererProperties]
  renderers [
   {rendererClass=&NiagaraSpriteRendererProperties}
   {rendererClass=&NiagaraMeshRendererProperties, rendererIndex=1}
  ]
 }
}
summary {
 systemName=NS_ApexWild_Helmet
 emitters [
  {
   emitterName=Arm_Glow
   rendererClasses [
    &NiagaraComponentRendererProperties
    &NiagaraLightRendererProperties
    &NiagaraMeshRendererProperties
   ]
  }
  {emitterName=Head_Glow, rendererClasses [&NiagaraSpriteRendererProperties, &NiagaraMeshRendererProperties]}
  {emitterName=Arm_Glow001, rendererClasses [&NiagaraMeshRendererProperties]}
 ]
 userVariables [
  User.IsEnabled {type=NiagaraFloat, defaultValue=1, description=""}
  User.SkeletalMesh {
   type=NiagaraDataInterfaceSkeletalMesh
   defaultValue {
    struct=NiagaraExt_VariableValue_DataInterface
    value {dataInterface=&self:NiagaraDataInterfaceSkeletalMesh_0}
   }
  }
 ]
}
