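# Physics asset for the male medium ApexWild body cosmetic, set up for loincloth simulation rather than a full-body ragdoll.
# Four capsule chains (front and back, left and right) each run four segments from a small base body linked to the pelvis, with limited swing and tight twist.
# Only pelvis, spine_01 and thighs represent the body, joined by free angular constraints, with capsules sized to the hips and upper legs.
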PhysicsAsset /BRCosmetics/Characters/Player/Male/Medium/Bodies/M_MED_ApexWild/Meshes/M_MED_Apex_Physics
bodies {
 loincloth_b_01_l [{shapeType=Capsule, length=8, radius=3, shapeName=loincloth_b_01_l_capsule, center=<3 0 0>}]
 loincloth_b_01_r [{shapeType=Capsule, length=7.4069, radius=3.5931, shapeName=loincloth_b_01_r_capsule, center=<-3 0 0>}]
 loincloth_b_02_l [{shapeType=Capsule, length=8, radius=3, shapeName=loincloth_b_02_l_capsule, center=<3 0 0>}]
 loincloth_b_02_r [{shapeType=Capsule, length=7.4069, radius=3.5931, shapeName=loincloth_b_02_r_capsule, center=<-3 0 0>, rotation=<rot 0 -0.0001 5>}]
 loincloth_b_03_l [{shapeType=Capsule, length=8, radius=3, shapeName=loincloth_b_03_l_capsule, center=<3 0 0>}]
 loincloth_b_03_r [{shapeType=Capsule, length=7.12, radius=3.88, shapeName=loincloth_b_03_r_capsule, center=<-3 0 0>, rotation=<rot 0 -0.0001 5>}]
 loincloth_b_04_l [{shapeType=Capsule, length=8, radius=3, shapeName=loincloth_b_04_l_capsule, center=<3 0 0>}]
 loincloth_b_04_r [{shapeType=Capsule, length=7.12, radius=3.88, shapeName=loincloth_b_04_r_capsule, center=<-3 0 0>, rotation=<rot 0 -0.0001 5>}]
 loincloth_b_base_l [{shapeType=Capsule, length=1, radius=1, shapeName=loincloth_b_base_l_capsule}]
 loincloth_b_base_r [{shapeType=Capsule, length=1, radius=1, shapeName=loincloth_b_base_r_capsule}]
 loincloth_fr_01_l [{shapeType=Capsule, length=8, radius=3, shapeName=loincloth_fr_01_l_capsule, center=<-3 0 0>}]
 loincloth_fr_01_r [{shapeType=Capsule, length=8, radius=3, shapeName=loincloth_fr_01_r_capsule, center=<3 0 0>}]
 loincloth_fr_02_l [{shapeType=Capsule, length=8.7535, radius=2.2465, shapeName=loincloth_fr_02_l_capsule, center=<-3 0 0>}]
 loincloth_fr_02_r [{shapeType=Capsule, length=14.5822, radius=2.2465, shapeName=loincloth_fr_02_r_capsule, center=<3 0 0>}]
 loincloth_fr_03_l [{shapeType=Capsule, length=11.4422, radius=2.2111, shapeName=loincloth_fr_03_l_capsule, center=<-3 0 0>}]
 loincloth_fr_03_r [{shapeType=Capsule, length=18.3736, radius=2.2111, shapeName=loincloth_fr_03_r_capsule, center=<3 0 3>}]
 loincloth_fr_04_l [{shapeType=Capsule, length=10.6533, radius=3, shapeName=loincloth_fr_04_l_capsule, center=<-3 0 0>}]
 loincloth_fr_04_r [{shapeType=Capsule, length=17.5847, radius=3, shapeName=loincloth_fr_04_r_capsule, center=<3 0 3>}]
 loincloth_fr_base_l [{shapeType=Capsule, length=1, radius=1, shapeName=loincloth_fr_base_l_capsule}]
 loincloth_fr_base_r [{shapeType=Capsule, length=1, radius=1, shapeName=loincloth_fr_base_r_capsule}]
 pelvis [{
   shapeType=Capsule
   length=15.4521
   radius=13.5309
   shapeName=pelvis_capsule
   center=<1.8619 -1.0285 -0.1294>
   rotation=<rot 1.7291 -90 0.6088>
  }]
 spine_01 [{
   shapeType=Capsule
   length=0.505
   radius=0.505
   shapeName=spine_01_capsule
   center=<10.1802 2.5727 -14.0517>
   rotation=<rot 0 -176.6495 -78.0696>
  }]
 thigh_l [{shapeType=Capsule, length=20.127, radius=8.123, shapeName=thigh_l_capsule, center=<-14 0 0>, rotation=<rot 90 0 0>}]
 thigh_r [{shapeType=Capsule, length=19.4774, radius=8.7726, shapeName=thigh_r_capsule, center=<15.5618 -1.2667 0>, rotation=<rot 90 0 0>}]
}
constraints [
 {bone1Name=loincloth_fr_01_r, bone2Name=loincloth_fr_base_r, swing1LimitDegrees=30, swing2LimitDegrees=30, twistLimitDegrees=5}
 {bone1Name=loincloth_fr_02_r, bone2Name=loincloth_fr_01_r, swing1LimitDegrees=30, swing2LimitDegrees=30, twistLimitDegrees=5}
 {bone1Name=loincloth_fr_03_r, bone2Name=loincloth_fr_02_r, swing1LimitDegrees=15, swing2LimitDegrees=15, twistLimitDegrees=5}
 {bone1Name=loincloth_fr_04_r, bone2Name=loincloth_fr_03_r, swing1LimitDegrees=15, swing2LimitDegrees=15, twistLimitDegrees=5}
 {bone1Name=loincloth_fr_01_l, bone2Name=loincloth_fr_base_l, swing1LimitDegrees=30, swing2LimitDegrees=5, twistLimitDegrees=5}
 {bone1Name=loincloth_fr_02_l, bone2Name=loincloth_fr_01_l, swing1LimitDegrees=30, swing2LimitDegrees=5, twistLimitDegrees=5}
 {bone1Name=loincloth_fr_03_l, bone2Name=loincloth_fr_02_l, swing1LimitDegrees=15, swing2LimitDegrees=5, twistLimitDegrees=5}
 {bone1Name=loincloth_fr_04_l, bone2Name=loincloth_fr_03_l, swing1LimitDegrees=15, swing2LimitDegrees=5, twistLimitDegrees=5}
 {bone1Name=loincloth_b_01_r, bone2Name=loincloth_b_base_r, swing1LimitDegrees=30, swing2LimitDegrees=30, twistLimitDegrees=5}
 {bone1Name=loincloth_b_02_r, bone2Name=loincloth_b_01_r, swing1LimitDegrees=30, swing2LimitDegrees=30, twistLimitDegrees=5}
 {bone1Name=loincloth_b_03_r, bone2Name=loincloth_b_02_r, swing1LimitDegrees=30, swing2LimitDegrees=30, twistLimitDegrees=5}
 {bone1Name=loincloth_b_04_r, bone2Name=loincloth_b_03_r, swing1LimitDegrees=30, swing2LimitDegrees=30, twistLimitDegrees=5}
 {bone1Name=loincloth_b_01_l, bone2Name=loincloth_b_base_l, swing1LimitDegrees=30, swing2LimitDegrees=30, twistLimitDegrees=5}
 {bone1Name=loincloth_b_02_l, bone2Name=loincloth_b_01_l, swing1LimitDegrees=30, swing2LimitDegrees=30, twistLimitDegrees=5}
 {bone1Name=loincloth_b_03_l, bone2Name=loincloth_b_02_l, swing1LimitDegrees=30, swing2LimitDegrees=30, twistLimitDegrees=5}
 {bone1Name=loincloth_b_04_l, bone2Name=loincloth_b_03_l, swing1LimitDegrees=30, swing2LimitDegrees=30, twistLimitDegrees=5}
 {
  bone1Name=thigh_l
  bone2Name=pelvis
  swing1LimitDegrees=45
  swing1Motion=Free
  swing2LimitDegrees=45
  swing2Motion=Free
  twistLimitDegrees=45
  twistMotion=Free
 }
 {
  bone1Name=thigh_r
  bone2Name=pelvis
  swing1LimitDegrees=45
  swing1Motion=Free
  swing2LimitDegrees=45
  swing2Motion=Free
  twistLimitDegrees=45
  twistMotion=Free
 }
 {
  bone1Name=loincloth_b_base_r
  bone2Name=pelvis
  swing1LimitDegrees=45
  swing1Motion=Free
  swing2LimitDegrees=45
  swing2Motion=Free
  twistLimitDegrees=45
  twistMotion=Free
 }
 {
  bone1Name=loincloth_b_base_l
  bone2Name=pelvis
  swing1LimitDegrees=45
  swing1Motion=Free
  swing2LimitDegrees=45
  swing2Motion=Free
  twistLimitDegrees=45
  twistMotion=Free
 }
 {
  bone1Name=loincloth_fr_base_r
  bone2Name=pelvis
  swing1LimitDegrees=45
  swing1Motion=Free
  swing2LimitDegrees=45
  swing2Motion=Free
  twistLimitDegrees=45
  twistMotion=Free
 }
 {
  bone1Name=loincloth_fr_base_l
  bone2Name=pelvis
  swing1LimitDegrees=45
  swing1Motion=Free
  swing2LimitDegrees=45
  swing2Motion=Free
  twistLimitDegrees=45
  twistMotion=Free
 }
 {
  bone1Name=spine_01
  bone2Name=pelvis
  swing1LimitDegrees=45
  swing1Motion=Free
  swing2LimitDegrees=45
  swing2Motion=Free
  twistLimitDegrees=45
  twistMotion=Free
 }
]
