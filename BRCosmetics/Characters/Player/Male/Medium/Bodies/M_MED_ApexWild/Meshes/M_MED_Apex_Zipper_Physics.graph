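# Physics asset for the zipper on the Apex Wild medium male player body, supporting physical motion of left and right zipper bone chains.
# Each chain has three capsule bodies linked by swing limits and tight twist limits, with its root attached to spine_05 by free angular motion.
# Capsules on spine_01, spine_04 and spine_05 form the torso volume, joined by angularly free constraints.
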PhysicsAsset /BRCosmetics/Characters/Player/Male/Medium/Bodies/M_MED_ApexWild/Meshes/M_MED_Apex_Zipper_Physics
bodies {
 spine_01 [{shapeType=Capsule, length=1, radius=1, shapeName=spine_01_capsule}]
 spine_04 [{shapeType=Capsule, length=16.6901, radius=13.3099, shapeName=spine_05_capsule, center=<6.4532 0 0>}]
 spine_05 [
  {shapeType=Capsule, length=16.5436, radius=10.8988, shapeName=spine_05_capsule}
  {shapeType=Capsule, length=17.4401, radius=12.5599, shapeName=spine_05_capsule, center=<-0.3895 -2.5675 0>}
 ]
 zipper_01_l [{shapeType=Capsule, length=1, radius=1, shapeName=zipper_01_l_capsule}]
 zipper_01_r [{shapeType=Capsule, length=1, radius=1, shapeName=zipper_01_r_capsule}]
 zipper_02_l [{shapeType=Capsule, length=2.0064, radius=2.2829, shapeName=zipper_02_l_capsule, center=<1.9347 0 0>, rotation=<rot 90 0 0>}]
 zipper_02_r [{shapeType=Capsule, length=2.452, radius=2.489, shapeName=zipper_02_r_capsule, center=<2.5122 0 0>, rotation=<rot 90 0 0>}]
 zipper_03_l [{shapeType=Capsule, length=1.1029, radius=2.1864, shapeName=zipper_03_l_capsule, center=<2.1408 0 0>, rotation=<rot 90 0 0>}]
 zipper_03_r [{shapeType=Capsule, length=2.2786, radius=2.6624, shapeName=zipper_03_r_capsule, center=<2.5122 0 0>, rotation=<rot 90 0 0>}]
}
constraints [
 {bone1Name=zipper_02_r, bone2Name=zipper_01_r, swing1LimitDegrees=30, swing2LimitDegrees=30, twistLimitDegrees=5}
 {bone1Name=zipper_03_r, bone2Name=zipper_02_r, swing1LimitDegrees=30, swing2LimitDegrees=30, twistLimitDegrees=5}
 {bone1Name=zipper_02_l, bone2Name=zipper_01_l, swing1LimitDegrees=30, swing2LimitDegrees=30, twistLimitDegrees=5}
 {bone1Name=zipper_03_l, bone2Name=zipper_02_l, swing1LimitDegrees=30, swing2LimitDegrees=30, twistLimitDegrees=5}
 {
  bone1Name=zipper_01_r
  bone2Name=spine_05
  swing1LimitDegrees=45
  swing1Motion=Free
  swing2LimitDegrees=45
  swing2Motion=Free
  twistLimitDegrees=45
  twistMotion=Free
 }
 {
  bone1Name=zipper_01_l
  bone2Name=spine_05
  swing1LimitDegrees=45
  swing1Motion=Free
  swing2LimitDegrees=45
  swing2Motion=Free
  twistLimitDegrees=45
  twistMotion=Free
 }
 {
  bone1Name=spine_05
  bone2Name=spine_04
  swing1LimitDegrees=45
  swing1Motion=Free
  swing2LimitDegrees=45
  swing2Motion=Free
  twistLimitDegrees=45
  twistMotion=Free
 }
 {
  bone1Name=spine_04
  bone2Name=spine_01
  swing1LimitDegrees=45
  swing1Motion=Free
  swing2LimitDegrees=45
  swing2Motion=Free
  twistLimitDegrees=45
  twistMotion=Free
 }
]
