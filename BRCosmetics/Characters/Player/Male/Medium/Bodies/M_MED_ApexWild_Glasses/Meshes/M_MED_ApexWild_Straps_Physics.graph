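# Physics asset for the straps of the male medium ApexWild Glasses cosmetic body, built entirely from capsule bodies.
# Left, center and right chains of seven strap segments hang from base bodies on the pelvis, alongside thigh and two-segment loincloth bodies.
# Strap joints mostly use tight swing and twist limits, with some axes locked near the roots.
# Pelvis attachments, loincloth joints and the right strap's root rotate freely.
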PhysicsAsset /BRCosmetics/Characters/Player/Male/Medium/Bodies/M_MED_ApexWild_Glasses/Meshes/M_MED_ApexWild_Straps_Physics
bodies {
 loincloth_fr_03_l [{shapeType=Capsule, length=9.755, radius=4.255, shapeName=loincloth_fr_03_l_capsule, center=<-3 1 0>}]
 loincloth_fr_03_r [{shapeType=Capsule, length=12.505, radius=4.255, shapeName=loincloth_fr_03_r_capsule, center=<3 -2 3.6131>}]
 loincloth_fr_04_l [{shapeType=Capsule, length=9.755, radius=4.255, shapeName=loincloth_fr_03_l_capsule, center=<-3 1 0>}]
 loincloth_fr_04_r [{shapeType=Capsule, length=12.255, radius=4.255, shapeName=loincloth_fr_03_r_capsule, center=<3 -2 4.1804>}]
 pelvis [{shapeType=Capsule, length=8.5, radius=10.8701, shapeName=pelvis_capsule, center=<0 -0.1412 0>}]
 strap_01_c [{shapeType=Capsule, length=6, radius=3, shapeName=strap_01_c_capsule, center=<3 0 0>, rotation=<rot 90 0 0>}]
 strap_01_c_base [{shapeType=Capsule, length=1, radius=1, shapeName=strap_01_c_base_capsule}]
 strap_01_l [{shapeType=Capsule, length=5.5, radius=3.5, shapeName=strap_01_l_capsule, center=<3 0 0>, rotation=<rot 90 0 0>}]
 strap_01_l_base [{shapeType=Capsule, length=1, radius=1, shapeName=strap_01_l_base_capsule}]
 strap_01_r [{shapeType=Capsule, length=6, radius=3, shapeName=strap_01_c_capsule, center=<3 0 0>, rotation=<rot 90 0 0>}]
 strap_01_r_base [{shapeType=Capsule, length=1, radius=1, shapeName=strap_01_r_base_capsule}]
 strap_02_c [{shapeType=Capsule, length=5.5, radius=2.5, shapeName=strap_02_c_capsule, center=<3 0 0>, rotation=<rot 90 0 0>}]
 strap_02_l [{shapeType=Capsule, length=5.5, radius=2.5, shapeName=strap_02_l_capsule, center=<3 0 0>, rotation=<rot 90 0 0>}]
 strap_02_r [{shapeType=Capsule, length=4.75, radius=2.5, shapeName=strap_03_l_capsule, center=<3 0 0>, rotation=<rot 90 0 0>}]
 strap_03_c [{shapeType=Capsule, length=5.5, radius=2.5, shapeName=strap_02_l_capsule, center=<3 0 0>, rotation=<rot 90 0 0>}]
 strap_03_l [{shapeType=Capsule, length=5.5, radius=2.5, shapeName=strap_02_l_capsule, center=<3 0 0>, rotation=<rot 90 0 0>}]
 strap_03_r [{shapeType=Capsule, length=4.75, radius=2.5, shapeName=strap_03_l_capsule, center=<3 0 0>, rotation=<rot 90 0 0>}]
 strap_04_c [{shapeType=Capsule, length=5.5, radius=2.5, shapeName=strap_02_l_capsule, center=<3 0 0>, rotation=<rot 90 0 0>}]
 strap_04_l [{shapeType=Capsule, length=5.5, radius=2.5, shapeName=strap_02_l_capsule, center=<3 0 0>, rotation=<rot 90 0 0>}]
 strap_04_r [{shapeType=Capsule, length=4.75, radius=2.5, shapeName=strap_03_l_capsule, center=<3 0 0>, rotation=<rot 90 0 0>}]
 strap_05_c [{shapeType=Capsule, length=5.5, radius=2.5, shapeName=strap_02_l_capsule, center=<3 0 0>, rotation=<rot 90 0 0>}]
 strap_05_l [{shapeType=Capsule, length=5.5, radius=2.5, shapeName=strap_02_l_capsule, center=<3 0 0>, rotation=<rot 90 0 0>}]
 strap_05_r [{shapeType=Capsule, length=4.75, radius=2.5, shapeName=strap_03_l_capsule, center=<3 0 0>, rotation=<rot 90 0 0>}]
 strap_06_c [{shapeType=Capsule, length=5.5, radius=2.5, shapeName=strap_02_l_capsule, center=<3 0 0>, rotation=<rot 90 0 0>}]
 strap_06_l [{shapeType=Capsule, length=5.5, radius=2.5, shapeName=strap_02_l_capsule, center=<3 0 0>, rotation=<rot 90 0 0>}]
 strap_06_r [{shapeType=Capsule, length=4.75, radius=2.5, shapeName=strap_03_l_capsule, center=<3 0 0>, rotation=<rot 90 0 0>}]
 strap_07_c [{shapeType=Capsule, length=5.5, radius=2.5, shapeName=strap_02_l_capsule, center=<3 0 0>, rotation=<rot 90 0 0>}]
 strap_07_l [{shapeType=Capsule, length=5.5, radius=2.5, shapeName=strap_02_l_capsule, center=<3 0 0>, rotation=<rot 90 0 0>}]
 strap_07_r [{shapeType=Capsule, length=4.75, radius=2.5, shapeName=strap_03_l_capsule, center=<3 0 0>, rotation=<rot 90 0 0>}]
 thigh_l [{shapeType=Capsule, length=26.5, radius=8.25, shapeName=thigh_l_capsule, center=<-19.238 0 0>, rotation=<rot 90 0 0>}]
 thigh_r [{shapeType=Capsule, length=26.5, radius=12, shapeName=thigh_r_capsule, center=<19.3628 0 0>, rotation=<rot 90 0 0>}]
}
constraints [
 {bone1Name=strap_01_c, bone2Name=strap_01_c_base, swing1LimitDegrees=15, swing2LimitDegrees=15, twistLimitDegrees=5}
 {
  bone1Name=strap_02_c
  bone2Name=strap_01_c
  swing1LimitDegrees=5
  swing2LimitDegrees=15
  swing2Motion=Locked
  twistLimitDegrees=5
  twistMotion=Locked
 }
 {bone1Name=strap_01_l, bone2Name=strap_01_l_base, swing1LimitDegrees=15, swing2LimitDegrees=15, twistLimitDegrees=5}
 {
  bone1Name=strap_01_r_base
  bone2Name=pelvis
  swing1LimitDegrees=45
  swing1Motion=Free
  swing2LimitDegrees=45
  swing2Motion=Free
  twistLimitDegrees=45
  twistMotion=Free
 }
 {
  bone1Name=strap_01_c_base
  bone2Name=pelvis
  swing1LimitDegrees=45
  swing1Motion=Free
  swing2LimitDegrees=45
  swing2Motion=Free
  twistLimitDegrees=45
  twistMotion=Free
 }
 {
  bone1Name=strap_01_l_base
  bone2Name=pelvis
  swing1LimitDegrees=45
  swing1Motion=Free
  swing2LimitDegrees=45
  swing2Motion=Free
  twistLimitDegrees=45
  twistMotion=Free
 }
 {
  bone1Name=thigh_l
  bone2Name=pelvis
  swing1LimitDegrees=45
  swing1Motion=Free
  swing2LimitDegrees=45
  swing2Motion=Free
  twistLimitDegrees=45
  twistMotion=Free
 }
 {
  bone1Name=thigh_r
  bone2Name=pelvis
  swing1LimitDegrees=45
  swing1Motion=Free
  swing2LimitDegrees=45
  swing2Motion=Free
  twistLimitDegrees=45
  twistMotion=Free
 }
 {
  bone1Name=loincloth_fr_03_l
  bone2Name=pelvis
  swing1LimitDegrees=45
  swing1Motion=Free
  swing2LimitDegrees=45
  swing2Motion=Free
  twistLimitDegrees=45
  twistMotion=Free
 }
 {
  bone1Name=loincloth_fr_03_r
  bone2Name=pelvis
  swing1LimitDegrees=45
  swing1Motion=Free
  swing2LimitDegrees=45
  swing2Motion=Free
  twistLimitDegrees=45
  twistMotion=Free
 }
 {
  bone1Name=loincloth_fr_04_l
  bone2Name=loincloth_fr_03_l
  swing1LimitDegrees=45
  swing1Motion=Free
  swing2LimitDegrees=45
  swing2Motion=Free
  twistLimitDegrees=45
  twistMotion=Free
 }
 {
  bone1Name=loincloth_fr_04_r
  bone2Name=loincloth_fr_03_r
  swing1LimitDegrees=45
  swing1Motion=Free
  swing2LimitDegrees=45
  swing2Motion=Free
  twistLimitDegrees=45
  twistMotion=Free
 }
 {bone1Name=strap_03_r, bone2Name=strap_02_r, swing1LimitDegrees=10, swing2LimitDegrees=15, twistLimitDegrees=5}
 {bone1Name=strap_04_r, bone2Name=strap_03_r, swing1LimitDegrees=15, swing2LimitDegrees=15, twistLimitDegrees=5}
 {bone1Name=strap_05_r, bone2Name=strap_04_r, swing1LimitDegrees=10, swing2LimitDegrees=15, twistLimitDegrees=5}
 {bone1Name=strap_06_r, bone2Name=strap_05_r, swing1LimitDegrees=10, swing2LimitDegrees=15, twistLimitDegrees=5}
 {bone1Name=strap_07_r, bone2Name=strap_06_r, swing1LimitDegrees=10, swing2LimitDegrees=15, twistLimitDegrees=5}
 {bone1Name=strap_03_c, bone2Name=strap_02_c, swing1LimitDegrees=15, swing2LimitDegrees=15, twistLimitDegrees=5}
 {bone1Name=strap_04_c, bone2Name=strap_03_c, swing1LimitDegrees=15, swing2LimitDegrees=15, twistLimitDegrees=5}
 {bone1Name=strap_05_c, bone2Name=strap_04_c, swing1LimitDegrees=15, swing2LimitDegrees=15, twistLimitDegrees=5}
 {bone1Name=strap_06_c, bone2Name=strap_05_c, swing1LimitDegrees=15, swing2LimitDegrees=15, twistLimitDegrees=5}
 {bone1Name=strap_07_c, bone2Name=strap_06_c, swing1LimitDegrees=15, swing2LimitDegrees=15, twistLimitDegrees=5}
 {bone1Name=strap_02_l, bone2Name=strap_01_l, swing1LimitDegrees=5, swing2LimitDegrees=0, swing2Motion=Locked, twistLimitDegrees=5}
 {bone1Name=strap_03_l, bone2Name=strap_02_l, swing1LimitDegrees=45, swing2LimitDegrees=15, twistLimitDegrees=5}
 {bone1Name=strap_04_l, bone2Name=strap_03_l, swing1LimitDegrees=15, swing2LimitDegrees=15, twistLimitDegrees=5}
 {bone1Name=strap_05_l, bone2Name=strap_04_l, swing1LimitDegrees=15, swing2LimitDegrees=15, twistLimitDegrees=5}
 {bone1Name=strap_06_l, bone2Name=strap_05_l, swing1LimitDegrees=15, swing2LimitDegrees=15, twistLimitDegrees=5}
 {bone1Name=strap_07_l, bone2Name=strap_06_l, swing1LimitDegrees=15, swing2LimitDegrees=15, twistLimitDegrees=5}
 {
  bone1Name=strap_02_r
  bone2Name=strap_01_r
  swing1LimitDegrees=5
  swing2LimitDegrees=0
  swing2Motion=Locked
  twistLimitDegrees=0
  twistMotion=Locked
 }
 {
  bone1Name=strap_01_r
  bone2Name=strap_01_r_base
  swing1LimitDegrees=45
  swing1Motion=Free
  swing2LimitDegrees=45
  swing2Motion=Free
  twistLimitDegrees=45
  twistMotion=Free
 }
]
