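# Post-process style AnimBlueprint for the ApexWild Helmet cosmetic, adding secondary motion to loincloth, straps, hood, zippers and armor.
# Copied parent pose runs through a deform Control Rig, crouch-switched RigidBody branches, then chained Stage 1–3 caches adding constraints, corrections and physics.
# Output uses Stage 3, or a constraint-based fallback without rigid bodies; variant-tag stage blends are unconnected.
# EventGraph derives corrective alphas from bone deltas and movement state.
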AnimBlueprint /BRCosmetics/Characters/Player/Male/Medium/Bodies/M_MED_ApexWild_Helmet/Meshes/Meshes/M_MED_ApexWild_Helmet_AnimBP
target_skeleton=/BRCosmetics/Characters/Player/Male/Medium/Bodies/M_MED_ApexWild/Meshes/M_MED_Apex_Skeleton
graphs [
 AnimGraph {nodes [
   Blend Poses by bool(BlendPose_1=Blend Poses by bool(bActiveValue=Get stage2() -> (stage2:bool)) {blendType=Linear}, bActiveValue=Get stage3() -> (stage3:bool)) {blendType=Linear}
   Output Pose(Result=Blend Poses by bool(BlendPose_0=Use cached pose 'Stage 3 Cache'(), BlendPose_1=Use cached pose 'RBAN Disabled Cache'(), bActiveValue=Get bRigidBodyEnabled()) {blendType=Linear}) {layerGroup=None}
   n1 = "Constraint - straps_parent (PPP)"!(
    ComponentPose=Local To Component(LocalPose=Use cached pose 'Stage 1 Cache'())
    Alpha=1
    ConstraintWeights_0=0
    ConstraintWeights_1:float=0.5
    ConstraintWeights_2:float=0.5
   ) -> (Pose) {
    alphaBoolBlend {blendOption=Linear}
    boneToModify {boneName=straps_parent}
    constraintSetup [{targetBone {boneName=pelvis}}, {targetBone {boneName=spine_02}}, {targetBone {boneName=spine_01}}]
    constraintWeights [1,1,1]
   }
   n2 = "Constraint - strap_01_r_base (R)"(ComponentPose=@n1, Alpha=0.75, ConstraintWeights_0=1) {
    alphaBoolBlend {blendOption=Linear}
    boneToModify {boneName=strap_01_r_base}
    constraintSetup [{transformType=Rotation, targetBone {boneName=loincloth_fr_02_r}}]
    constraintWeights [1]
   }
   n3 = "Constraint - strap_01_l (R)"(ComponentPose=@n2, Alpha=0.75, ConstraintWeights_0=1) {
    alphaBoolBlend {blendOption=Linear}
    boneToModify {boneName=strap_01_l}
    constraintSetup [{transformType=Rotation, targetBone {boneName=loincloth_fr_02_l}}]
    constraintWeights [1]
   }
   n4 = "Constraint - strap_01_c (RR)"!(ComponentPose=@n3, Alpha=0.75, ConstraintWeights_0=0.6, ConstraintWeights_1:float=0.4) -> (Pose) {
    alphaBoolBlend {blendOption=Linear}
    boneToModify {boneName=strap_01_c}
    constraintSetup [{transformType=Rotation, targetBone {boneName=loincloth_fr_02_r}}, {transformType=Rotation, targetBone {boneName=loincloth_fr_02_l}}]
    constraintWeights [1,1]
   }
   n5 = "Transform (Modify) Bone - Bone: strap_01_c_base"(ComponentPose=@n4, Alpha=Get thighLCrouchAlpha() -> (thighLCrouchAlpha:double), Translation="0,0,0", Rotation="0,30,0", Scale="1,1,1") {
    rotationMode=BMM_Additive
    rotationSpace=BCS_BoneSpace
    translationMode=BMM_Additive
    translationSpace=BCS_BoneSpace
    alphaBoolBlend {blendInTime=0.25, blendOption=Linear, blendOutTime=0.25}
    alphaScaleBiasClamp {bInterpResult=true, inRange {max=0, min=1}}
    boneToModify {boneName=strap_01_c_base}
    rotation {yaw=30}
   }
   n6 = "Transform (Modify) Bone - Bone: strap_01_c_base"(ComponentPose=@n5, Alpha=Get thighRAlpha() -> (thighRAlpha:double), Translation="0,0,0", Rotation="0,50,0", Scale="1,1,1") {
    rotationMode=BMM_Additive
    rotationSpace=BCS_BoneSpace
    translationMode=BMM_Additive
    translationSpace=BCS_BoneSpace
    alphaBoolBlend {blendInTime=0.25, blendOption=Linear, blendOutTime=0.25}
    alphaScaleBiasClamp {bInterpResult=true, inRange {max=0, min=1}}
    boneToModify {boneName=strap_01_c_base}
    rotation {yaw=50}
   }
   n7 = "Transform (Modify) Bone - Bone: strap_01_r_base"(ComponentPose=@n6, Alpha=Get thighRAlpha(), Translation="0,0,0", Rotation="0,60,0", Scale="1,1,1") {
    rotationMode=BMM_Additive
    rotationSpace=BCS_BoneSpace
    translationMode=BMM_Additive
    translationSpace=BCS_BoneSpace
    alphaBoolBlend {blendInTime=0.25, blendOption=Linear, blendOutTime=0.25}
    alphaScaleBiasClamp {bInterpResult=true}
    boneToModify {boneName=strap_01_r_base}
    rotation {yaw=60}
   }
   n8 = "Transform (Modify) Bone - Bone: strap_01_l_base"(ComponentPose=@n7, Alpha=Get thighLAlpha() -> (thighLAlpha:double), Translation="0,0,0", Rotation="0,30,0", Scale="1,1,1") {
    rotationMode=BMM_Additive
    rotationSpace=BCS_BoneSpace
    translationMode=BMM_Additive
    translationSpace=BCS_BoneSpace
    alphaBoolBlend {blendInTime=0.25, blendOption=Linear, blendOutTime=0.25}
    alphaScaleBiasClamp {bInterpResult=true}
    boneToModify {boneName=strap_01_l_base}
    rotation {yaw=30}
   }
   n9 = "Transform (Modify) Bone - Bone: strap_01_l_base"(ComponentPose=@n8, Alpha=Get thighLCrouchAlpha(), Translation="0,0,0", Rotation="0,30,0", Scale="1,1,1") {
    rotationMode=BMM_Additive
    rotationSpace=BCS_BoneSpace
    translationMode=BMM_Additive
    translationSpace=BCS_BoneSpace
    alphaBoolBlend {blendInTime=0.25, blendOption=Linear, blendOutTime=0.25}
    alphaScaleBiasClamp {bInterpResult=true}
    boneToModify {boneName=strap_01_l_base}
    rotation {yaw=30}
   }
   n10 = "Transform (Modify) Bone - Bone: strap_01_l_base"(ComponentPose=@n9, Alpha=Get thighRCrouchAlpha() -> (thighRCrouchAlpha:double), Translation="0,0,0", Rotation="0,30,0", Scale="1,1,1") {
    rotationMode=BMM_Additive
    rotationSpace=BCS_BoneSpace
    translationMode=BMM_Additive
    translationSpace=BCS_BoneSpace
    alphaBoolBlend {blendInTime=0.25, blendOption=Linear, blendOutTime=0.25}
    alphaScaleBiasClamp {bInterpResult=true}
    boneToModify {boneName=strap_01_l_base}
    rotation {yaw=30}
   }
   n11 = Get jump_alpha() -> (jump_alpha:double)
   n12 = "Transform (Modify) Bone - Bone: strap_01_r_base"(ComponentPose=@n10, Alpha=@n11, Translation="0,0,0", Rotation="0,-30,0", Scale="1,1,1") {
    rotationMode=BMM_Additive
    rotationSpace=BCS_BoneSpace
    translationMode=BMM_Additive
    translationSpace=BCS_BoneSpace
    alphaBoolBlend {blendInTime=0.25, blendOption=Linear, blendOutTime=0.25}
    alphaScaleBiasClamp {bInterpResult=true}
    boneToModify {boneName=strap_01_r_base}
    rotation {yaw=-30}
   }
   n13 = "Transform (Modify) Bone - Bone: strap_01_c_base"(ComponentPose=@n12, Alpha=@n11, Translation="0,0,0", Rotation="0,-15,0", Scale="1,1,1") {
    rotationMode=BMM_Additive
    rotationSpace=BCS_BoneSpace
    translationMode=BMM_Additive
    translationSpace=BCS_BoneSpace
    alphaBoolBlend {blendInTime=0.25, blendOption=Linear, blendOutTime=0.25}
    alphaScaleBiasClamp {bInterpResult=true}
    boneToModify {boneName=strap_01_c_base}
    rotation {yaw=-15}
   }
   n14 = RigidBody(ComponentPose=@n13) {
    bOverrideWorldGravity=true
    overridePhysicsAsset=/BRCosmetics/Characters/Player/Male/Medium/Bodies/M_MED_ApexWild_Glasses/Meshes/M_MED_ApexWild_Straps_Physics
    simulationSpace=BaseBoneSpace
    alphaBoolBlend {blendOption=Linear}
    baseBoneRef {boneName=pelvis}
    componentAppliedLinearAccClamp=<450 450 650>
    componentLinearAccScale=<0.075 0.075 0.1>
    componentLinearVelScale=<0.075 0.075 0.1>
    overrideWorldGravity {z=-600}
    simSpaceSettings {maxAngularAcceleration=1000, maxAngularVelocity=3, maxLinearAcceleration=250, maxLinearVelocity=500, externalLinearDragV {y=0.5}}
   }
   n15 = "Transform (Modify) Bone - Bone: loincloth_fr_base_l"!(
    ComponentPose=@n14
    bAlphaBoolEnabled:bool=Get crouch_not_moving() -> (crouch_not_moving:bool)
    Translation="0,0,0"
    Rotation="-10,0.000,0.000"
    Scale="1,1,1"
   ) -> (Pose) {
    alphaInputType=Bool
    rotationMode=BMM_Additive
    rotationSpace=BCS_BoneSpace
    translationMode=BMM_Additive
    translationSpace=BCS_BoneSpace
    alphaBoolBlend {blendInTime=0.25, blendOption=Linear, blendOutTime=0.25}
    alphaScaleBiasClamp {bInterpResult=true}
    boneToModify {boneName=loincloth_fr_base_l}
    rotation {pitch=-10}
   }
   Save cached pose 'Stage 2 Cache'(Pose=Component To Local(ComponentPose=@n15)) {cacheName=Stage 2 Cache}
   n16 = Control Rig(Source=Copy Pose From Mesh() {bUseAttachedParent=true}) {
    alphaBoolBlend {blendOption=Linear}
    controlRigAssetReference {blueprintRigClass=/BRCosmetics/Characters/Player/Male/Medium/Bodies/M_MED_ApexWild/Meshes/M_MED_DeformRig_ApexWild_ControlRig.M_MED_DeformRig_ApexWild_ControlRig_C}
   }
   n17 = "Transform (Modify) Bone - Bone: loincloth_fr_base_r"(
    ComponentPose=Local To Component(LocalPose=@n16)
    Alpha=Get thighRAlpha()
    Translation="-4.610,1.995,-0.000"
    Rotation="0.000,100.000,-15.000"
    Scale="1,1,1"
   ) {
    rotationMode=BMM_Additive
    rotationSpace=BCS_BoneSpace
    translationMode=BMM_Additive
    translationSpace=BCS_BoneSpace
    alphaBoolBlend {blendOption=Linear}
    alphaScaleBiasClamp {interpSpeedIncreasing=5}
    boneToModify {boneName=loincloth_fr_base_r}
    rotation=<rot 0 100.0001 -15>
    translation=<-4.61 1.995>
   }
   n18 = "Transform (Modify) Bone - Bone: loincloth_fr_base_l"!(ComponentPose=@n17, Alpha:float=Get thighLAlpha(), Translation="2,0.000,0.000", Rotation="0.000,90.000,-15.000", Scale="1,1,1") -> (Pose) {
    rotationMode=BMM_Additive
    rotationSpace=BCS_BoneSpace
    translationMode=BMM_Additive
    translationSpace=BCS_BoneSpace
    alphaBoolBlend {blendOption=Linear}
    boneToModify {boneName=loincloth_fr_base_l}
    rotation=<rot 0 90.0001 -15>
    translation {x=2}
   }
   n19 = RigidBody!(ComponentPose=@n18, Alpha, OverrideWorldGravity:Vector=Get GravityOverride(), ExternalForce) -> (Pose) {
    bOverrideWorldGravity=true
    overridePhysicsAsset=/BRCosmetics/Characters/Player/Male/Medium/Bodies/M_MED_ApexWild/Meshes/M_MED_Apex_Physics
    simulationSpace=BaseBoneSpace
    alphaBoolBlend {blendOption=Linear}
    baseBoneRef {boneName=pelvis}
    componentAppliedLinearAccClamp=<1500 1500 2000>
    componentLinearAccScale=<0.25 0.25 0.3>
    componentLinearVelScale=<0.25 0.25 0.3>
    overrideWorldGravity {z=-980}
    simSpaceSettings {maxAngularAcceleration=1000, maxAngularVelocity=3, maxLinearAcceleration=250, maxLinearVelocity=500, externalLinearDragV {y=0.5}}
   }
   n20 = Get jump_alpha()
   n21 = "Transform (Modify) Bone - Bone: loincloth_fr_base_r"(ComponentPose=@n19, Alpha=@n20, Translation="-7,5,0.3", Rotation="10,-30,-5.284", Scale="1,1,1") {
    rotationMode=BMM_Additive
    rotationSpace=BCS_BoneSpace
    translationMode=BMM_Additive
    translationSpace=BCS_BoneSpace
    alphaBoolBlend {blendOption=Linear}
    alphaScaleBiasClamp {bInterpResult=true}
    boneToModify {boneName=loincloth_fr_base_r}
    rotation=<rot 10 -30 -5.284>
    translation=<-7 5 0.3>
   }
   n22 = "Transform (Modify) Bone - Bone: deform_thigh_helper_r"(ComponentPose=@n21, Alpha=@n20, Translation="0,0,2.5", Rotation="0,0,0", Scale="1,1,1") {
    rotationMode=BMM_Additive
    rotationSpace=BCS_BoneSpace
    translationMode=BMM_Additive
    translationSpace=BCS_BoneSpace
    alphaBoolBlend {blendInTime=0.25, blendOption=Linear, blendOutTime=0.25}
    boneToModify {boneName=deform_thigh_helper_r}
    translation {z=2.5}
   }
   n23 = Component To Local(ComponentPose=@n22)
   n24 = Control Rig(Source=Copy Pose From Mesh() {bUseAttachedParent=true}) {
    alphaBoolBlend {blendOption=Linear}
    controlRigAssetReference {blueprintRigClass=&M_MED_DeformRig_ApexWild_ControlRig_C}
   }
   n25 = "Transform (Modify) Bone - Bone: loincloth_fr_base_l"(ComponentPose=Local To Component(LocalPose=@n24), Alpha=Get thighLAlpha(), Translation="9,2.75,0.000", Rotation="0,0,0", Scale="1,1,1") {
    rotationMode=BMM_Additive
    rotationSpace=BCS_BoneSpace
    translationMode=BMM_Additive
    translationSpace=BCS_BoneSpace
    alphaBoolBlend {blendOption=Linear}
    alphaScaleBiasClamp {bInterpResult=true}
    boneToModify {boneName=loincloth_fr_base_l}
    translation=<9 2.75>
   }
   n26 = "Transform (Modify) Bone - Bone: loincloth_fr_base_r"(ComponentPose=@n25, Alpha=Get thighRCrouchAlpha(), Translation="-7.217,5.197,0", Rotation="0.000,90,0.000", Scale="1,1,1") {
    rotationMode=BMM_Additive
    rotationSpace=BCS_BoneSpace
    translationMode=BMM_Additive
    translationSpace=BCS_BoneSpace
    alphaBoolBlend {blendOption=Linear}
    alphaScaleBiasClamp {bInterpResult=true}
    boneToModify {boneName=loincloth_fr_base_r}
    rotation {yaw=90}
    translation=<-7.217 5.197>
   }
   n27 = "Transform (Modify) Bone - Bone: loincloth_fr_base_l"(ComponentPose=@n26, Alpha=Get thighLAlpha(), Translation="0,0,0", Rotation="0.000,90,0.000", Scale="1,1,1") {
    rotationMode=BMM_Additive
    rotationSpace=BCS_BoneSpace
    translationMode=BMM_Additive
    translationSpace=BCS_BoneSpace
    alphaBoolBlend {blendOption=Linear}
    alphaScaleBiasClamp {bInterpResult=true}
    boneToModify {boneName=loincloth_fr_base_l}
    rotation {yaw=90}
   }
   n28 = RigidBody!(ComponentPose=@n27, Alpha, ExternalForce) -> (Pose) {
    bOverrideWorldGravity=true
    overridePhysicsAsset=/BRCosmetics/Characters/Player/Male/Medium/Bodies/M_MED_ApexWild/Meshes/M_MED_Apex_Crouch_Physics
    simulationSpace=BaseBoneSpace
    alphaBoolBlend {blendOption=Linear}
    baseBoneRef {boneName=pelvis}
    componentAppliedLinearAccClamp=<1500 1500 2000>
    componentLinearAccScale=<0.25 0.25 0.3>
    componentLinearVelScale=<0.25 0.25 0.3>
    overrideWorldGravity {z=-980}
    simSpaceSettings {maxAngularAcceleration=1000, maxAngularVelocity=3, maxLinearAcceleration=250, maxLinearVelocity=500, externalLinearDragV {y=0.5}}
   }
   n29 = RigidBody(ComponentPose=@n28) {
    bOverrideWorldGravity=true
    overridePhysicsAsset=/BRCosmetics/Characters/Player/Male/Medium/Bodies/M_MED_ApexWild/Meshes/M_MED_Apex_Zipper_Physics
    simulationSpace=BaseBoneSpace
    alphaBoolBlend {blendOption=Linear}
    baseBoneRef {boneName=spine_05}
    componentAppliedLinearAccClamp=<2000 2000 2000>
    componentLinearAccScale=<0.3 0.3 0.35>
    componentLinearVelScale=<0.3 0.3 0.35>
    overrideWorldGravity {z=-980}
    simSpaceSettings {maxAngularAcceleration=1000, maxAngularVelocity=3, maxLinearAcceleration=250, maxLinearVelocity=500, worldAlpha=1}
   }
   n30 = "Transform (Modify) Bone - Bone: loincloth_fr_base_r"(ComponentPose=@n29, Alpha=Get thighRAlpha(), Translation="-4.75,4.75,0", Rotation="10,0,0", Scale="1,1,1") {
    rotationMode=BMM_Additive
    rotationSpace=BCS_BoneSpace
    translationMode=BMM_Additive
    translationSpace=BCS_BoneSpace
    alphaBoolBlend {blendInTime=0.25, blendOption=Linear, blendOutTime=0.25}
    alphaScaleBiasClamp {bInterpResult=true}
    boneToModify {boneName=loincloth_fr_base_r}
    rotation {pitch=10}
    translation=<-4.75 4.75>
   }
   n31 = "Transform (Modify) Bone - Bone: loincloth_fr_base_l"(ComponentPose=@n30, Alpha=Get thighRAlpha(), Translation="0,0.5,0", Rotation="0,0,-20", Scale="1,1,1") {
    rotationMode=BMM_Additive
    rotationSpace=BCS_BoneSpace
    translationMode=BMM_Additive
    translationSpace=BCS_BoneSpace
    alphaBoolBlend {blendInTime=0.25, blendOption=Linear, blendOutTime=0.25}
    alphaScaleBiasClamp {bInterpResult=true, bMapRange=true, inRange {max=0, min=1}}
    boneToModify {boneName=loincloth_fr_base_l}
    rotation {roll=-20}
    translation {y=0.5}
   }
   n32 = "Transform (Modify) Bone - Bone: deform_thigh_helper_r"(ComponentPose=@n31, Alpha=Get thighRCrouchAlpha(), Translation="0,0,4", Rotation="0,0,0", Scale="1,1,1") {
    rotationMode=BMM_Additive
    rotationSpace=BCS_BoneSpace
    translationMode=BMM_Additive
    translationSpace=BCS_BoneSpace
    alphaBoolBlend {blendInTime=0.25, blendOption=Linear, blendOutTime=0.25}
    boneToModify {boneName=deform_thigh_helper_r}
    translation {z=4}
   }
   n33 = Component To Local(ComponentPose=@n32)
   n34 = Get bIsCrouching()
   n35 = Get head() -> (head:double)
   n36 = "Transform (Modify) Bone - Bone: dyn_hood_base"(
    ComponentPose=Local To Component(LocalPose=Blend Poses by bool(BlendPose_0=@n33, BlendPose_1=@n23, bActiveValue=@n34) {blendType=Linear})
    Alpha=@n35
    Translation="0,0,0"
    Rotation="-0.000,-20,0.000"
    Scale="1,1,1"
   ) {
    rotationMode=BMM_Additive
    rotationSpace=BCS_BoneSpace
    alphaBoolBlend {blendOption=Linear}
    alphaScaleBiasClamp {bInterpResult=true}
    boneToModify {boneName=dyn_hood_base}
    rotation {yaw=-20}
   }
   n37 = "Transform (Modify) Bone - Bone: dyn_hood_right"(ComponentPose=@n36, Alpha=@n35, Translation="-1,0,0.000", Rotation="0,0,0", Scale="1,1,1") {
    rotationMode=BMM_Additive
    rotationSpace=BCS_BoneSpace
    translationMode=BMM_Additive
    translationSpace=BCS_BoneSpace
    alphaBoolBlend {blendOption=Linear}
    alphaScaleBiasClamp {bInterpResult=true}
    boneToModify {boneName=dyn_hood_right}
    translation {x=-1}
   }
   n38 = "Transform (Modify) Bone - Bone: dyn_hood_l_raise"(ComponentPose=@n37, Alpha=Get armUpLeftAlpha() -> (armUpLeftAlpha:double), Translation="0,0,0", Rotation="15,0,0", Scale="1,1,1") {
    rotationMode=BMM_Additive
    rotationSpace=BCS_BoneSpace
    alphaBoolBlend {blendOption=Linear}
    boneToModify {boneName=dyn_hood_l_raise}
    rotation {pitch=15}
   }
   n39 = RigidBody(ComponentPose=@n38) {
    bOverrideWorldGravity=true
    overridePhysicsAsset=&M_MED_Apex_Zipper_Physics
    simulationSpace=BaseBoneSpace
    alphaBoolBlend {blendOption=Linear}
    baseBoneRef {boneName=spine_05}
    componentAppliedLinearAccClamp=<2000 2000 2000>
    componentLinearAccScale=<0.3 0.3 0.35>
    componentLinearVelScale=<0.3 0.3 0.35>
    overrideWorldGravity {z=-980}
    simSpaceSettings {maxAngularAcceleration=1000, maxAngularVelocity=3, maxLinearAcceleration=250, maxLinearVelocity=500, worldAlpha=1}
   }
   n40 = Anim Dynamics - Bone: zipper_b_02(ComponentPose=@n39, Alpha=1, ExternalForce="0,0,0") {
    bUseGravityOverride=true
    simulationSpace=RootRelative
    alphaBoolBlend {blendOption=Linear}
    boundBone {boneName=zipper_b_02}
    componentAppliedLinearAccClamp=<1750 1750 1750>
    componentLinearAccScale=<0.25 0.25 0.325>
    componentLinearVelScale=<0.25 0.25 0.325>
    gravityOverride {z=-980}
    physicsBodyDefinitions [{
      boundBone {boneName=zipper_b_02}
      boxExtents=<7 7 7>
      constraintSetup {angularLimitsMax=<5 5 5>, angularLimitsMin=<-5 -5 -60>, angularTarget=<0 0 0>}
      localJointOffset=<5 0 0>
     }]
   }
   n41 = Anim Dynamics - Bone: zipper_b_03(ComponentPose=@n40, Alpha=1, ExternalForce="0,0,0") {
    bUseGravityOverride=true
    simulationSpace=RootRelative
    alphaBoolBlend {blendOption=Linear}
    boundBone {boneName=zipper_b_03}
    componentAppliedLinearAccClamp=<1750 1750 1750>
    componentLinearAccScale=<0.25 0.25 0.325>
    componentLinearVelScale=<0.25 0.25 0.325>
    gravityOverride {z=-980}
    physicsBodyDefinitions [{
      boundBone {boneName=zipper_b_03}
      boxExtents=<7 7 7>
      constraintSetup {angularLimitsMax=<5 5 5>, angularLimitsMin=<-5 -5 -60>, angularTarget=<0 0 0>}
      localJointOffset=<5 0 0>
     }]
   }
   n42 = "Constraint - zipper_b_01 (PPPP)"!(
    ComponentPose=@n41
    Alpha=1
    ConstraintWeights_0=0.55
    ConstraintWeights_1:float=0.15
    ConstraintWeights_2:float=0.15
    ConstraintWeights_3:float=0.15
   ) -> (Pose) {
    alphaBoolBlend {blendOption=Linear}
    boneToModify {boneName=zipper_b_01}
    constraintSetup [{targetBone {boneName=pelvis}}, {targetBone {boneName=spine_01}}, {targetBone {boneName=spine_02}}, {targetBone {boneName=spine_03}}]
    constraintWeights [1,1,1,1]
   }
   n43 = "Transform (Modify) Bone - Bone: zipper_b_01"(ComponentPose=@n42, Alpha=1, Translation="0,0,0", Rotation="0,-10,0", Scale="1,1,1") {
    rotationMode=BMM_Additive
    rotationSpace=BCS_BoneSpace
    alphaBoolBlend {blendOption=Linear}
    boneToModify {boneName=zipper_b_01}
    rotation {yaw=-10}
   }
   Save cached pose 'Stage 1 Cache'(Pose=Component To Local(ComponentPose=@n43)) {cacheName=Stage 1 Cache}
   n44 = Get head()
   n45 = "Transform (Modify) Bone - Bone: dyn_hood_base"(
    ComponentPose=Local To Component(LocalPose=Use cached pose 'Stage 2 Cache'())
    Alpha=@n44
    Translation="0,0,0"
    Rotation="-0.000,-5,0.000"
    Scale="1,1,1"
   ) {
    rotationMode=BMM_Additive
    rotationSpace=BCS_BoneSpace
    alphaBoolBlend {blendOption=Linear}
    alphaScaleBiasClamp {bInterpResult=true}
    boneToModify {boneName=dyn_hood_base}
    rotation {yaw=-5}
   }
   n46 = "Transform (Modify) Bone - Bone: dyn_hood_right"(ComponentPose=@n45, Alpha=@n44, Translation="-1,0,0.000", Rotation="0,0,0", Scale="1,1,1") {
    rotationMode=BMM_Additive
    rotationSpace=BCS_BoneSpace
    translationMode=BMM_Additive
    translationSpace=BCS_BoneSpace
    alphaBoolBlend {blendOption=Linear}
    alphaScaleBiasClamp {bInterpResult=true}
    boneToModify {boneName=dyn_hood_right}
    translation {x=-1}
   }
   n47 = "Transform (Modify) Bone - Bone: lowerarm_armor_r"(ComponentPose=@n46, Alpha=Get lowerarmRAlpha() -> (lowerarmRAlpha:double), Translation="0,0,0", Rotation="0,0.000,-45.000", Scale="1,1,1") {
    rotationMode=BMM_Additive
    rotationSpace=BCS_BoneSpace
    translationMode=BMM_Additive
    translationSpace=BCS_BoneSpace
    alphaBoolBlend {blendOption=Linear}
    boneToModify {boneName=lowerarm_armor_r}
    rotation {roll=-45}
   }
   n48 = "Constraint - armor_upper_r (PP)"!(ComponentPose=@n47, Alpha=1, ConstraintWeights_0=0.65, ConstraintWeights_1:float=0.35) -> (Pose) {
    alphaBoolBlend {blendOption=Linear}
    boneToModify {boneName=armor_upper_r}
    constraintSetup [{targetBone {boneName=upperarm_r}}, {targetBone {boneName=deform_upperarm_up_r}}]
    constraintWeights [1,1]
   }
   n49 = "Transform (Modify) Bone - Bone: armor_upper_r"!(
    ComponentPose=@n48
    bAlphaBoolEnabled:bool=Get frontend_not_emote() -> (frontend_not_emote:bool)
    Translation="-4,0,-2"
    Rotation="0,-5,10"
    Scale="1,1,1"
   ) -> (Pose) {
    alphaInputType=Bool
    rotationMode=BMM_Additive
    rotationSpace=BCS_BoneSpace
    translationMode=BMM_Additive
    translationSpace=BCS_BoneSpace
    alphaBoolBlend {blendOption=Linear}
    boneToModify {boneName=armor_upper_r}
    rotation {roll=10, yaw=-5}
    translation {x=-4, z=-2}
   }
   n50 = "Transform (Modify) Bone - Bone: armor_upper_r"!(ComponentPose=@n49, Alpha:float=1, Translation="0,0,0", Rotation="0,0,0", Scale="1,1,1") -> (Pose) {
    rotationMode=BMM_Additive
    rotationSpace=BCS_BoneSpace
    translationMode=BMM_Additive
    translationSpace=BCS_BoneSpace
    alphaBoolBlend {blendOption=Linear}
    boneToModify {boneName=armor_upper_r}
   }
   Save cached pose 'Stage 3 Cache'(Pose=Component To Local(ComponentPose=@n50)) {cacheName=Stage 3 Cache}
   n51 = Control Rig(Source=Copy Pose From Mesh() {bUseAttachedParent=true}) {
    alphaBoolBlend {blendOption=Linear}
    controlRigAssetReference {blueprintRigClass=&M_MED_DeformRig_ApexWild_ControlRig_C}
   }
   n52 = "Constraint - loincloth_fr_base_r (R)"(ComponentPose=Local To Component(LocalPose=@n51), Alpha=1, ConstraintWeights_0=1) {
    alphaBoolBlend {blendOption=Linear}
    boneToModify {boneName=loincloth_fr_base_r}
    constraintSetup [{transformType=Rotation, targetBone {boneName=thigh_r}}]
    constraintWeights [1]
   }
   n53 = "Constraint - loincloth_fr_base_l (R)"(ComponentPose=@n52, Alpha=1, ConstraintWeights_0=1) {
    alphaBoolBlend {blendOption=Linear}
    boneToModify {boneName=loincloth_fr_base_l}
    constraintSetup [{transformType=Rotation, targetBone {boneName=thigh_l}}]
    constraintWeights [1]
   }
   n54 = "Constraint - straps_parent (PPP)"(ComponentPose=@n53, Alpha=1, ConstraintWeights_0=0, ConstraintWeights_1=0.5, ConstraintWeights_2=0.5) {
    alphaBoolBlend {blendOption=Linear}
    boneToModify {boneName=straps_parent}
    constraintSetup [{targetBone {boneName=pelvis}}, {targetBone {boneName=spine_02}}, {targetBone {boneName=spine_01}}]
    constraintWeights [1,1,1]
   }
   n55 = "Constraint - strap_01_r_base (R)"(ComponentPose=@n54, Alpha=1, ConstraintWeights_0=1) {
    alphaBoolBlend {blendOption=Linear}
    boneToModify {boneName=strap_01_r_base}
    constraintSetup [{transformType=Rotation, targetBone {boneName=loincloth_fr_01_r}}]
    constraintWeights [1]
   }
   n56 = "Constraint - strap_01_l (R)"(ComponentPose=@n55, Alpha=0.75, ConstraintWeights_0=1) {
    alphaBoolBlend {blendOption=Linear}
    boneToModify {boneName=strap_01_l}
    constraintSetup [{transformType=Rotation, targetBone {boneName=loincloth_fr_02_l}}]
    constraintWeights [1]
   }
   n57 = "Constraint - strap_01_c (RR)"(ComponentPose=@n56, Alpha=1, ConstraintWeights_0=0.6, ConstraintWeights_1=0.4) {
    alphaBoolBlend {blendOption=Linear}
    boneToModify {boneName=strap_01_c}
    constraintSetup [{transformType=Rotation, targetBone {boneName=loincloth_fr_01_r}}, {transformType=Rotation, targetBone {boneName=loincloth_fr_01_l}}]
    constraintWeights [1,1]
   }
   n58 = "Transform (Modify) Bone - Bone: loincloth_fr_base_l"(ComponentPose=@n57, Alpha=Get thighLAlpha(), Translation="0,-7,0.000", Rotation="-15,-10,3.269", Scale="1,1,1") {
    rotationMode=BMM_Additive
    rotationSpace=BCS_BoneSpace
    translationMode=BMM_Additive
    translationSpace=BCS_BoneSpace
    alphaBoolBlend {blendOption=Linear}
    alphaScaleBiasClamp {bInterpResult=true}
    boneToModify {boneName=loincloth_fr_base_l}
    rotation=<rot -15 -10 3.269>
    translation {y=-7}
   }
   n59 = "Transform (Modify) Bone - Bone: loincloth_fr_base_r"(ComponentPose=@n58, Alpha=Get thighRCrouchAlpha(), Translation="-5,5,0", Rotation="0,0,0", Scale="1,1,1") {
    rotationMode=BMM_Additive
    rotationSpace=BCS_BoneSpace
    translationMode=BMM_Additive
    translationSpace=BCS_BoneSpace
    alphaBoolBlend {blendOption=Linear}
    alphaScaleBiasClamp {bInterpResult=true}
    boneToModify {boneName=loincloth_fr_base_r}
    translation=<-5 5>
   }
   n60 = "Transform (Modify) Bone - Bone: strap_01_l_base"(ComponentPose=@n59, Alpha=Get thighRAlpha(), Translation="0,0,0", Rotation="0,30,0", Scale="1,1,1") {
    rotationMode=BMM_Additive
    rotationSpace=BCS_BoneSpace
    translationMode=BMM_Additive
    translationSpace=BCS_BoneSpace
    alphaBoolBlend {blendInTime=0.25, blendOption=Linear, blendOutTime=0.25}
    alphaScaleBiasClamp {bInterpResult=true}
    boneToModify {boneName=strap_01_l_base}
    rotation {yaw=30}
   }
   n61 = "Transform (Modify) Bone - Bone: loincloth_fr_base_r"(ComponentPose=@n60, Alpha=1, Translation="0,1.5,0", Rotation="0,0,0", Scale="1,1,1") {
    rotationMode=BMM_Additive
    rotationSpace=BCS_BoneSpace
    translationMode=BMM_Additive
    translationSpace=BCS_BoneSpace
    alphaBoolBlend {blendOption=Linear}
    alphaScaleBiasClamp {bInterpResult=true}
    boneToModify {boneName=loincloth_fr_base_r}
    translation {y=1.5}
   }
   n62 = Get head()
   n63 = "Transform (Modify) Bone - Bone: dyn_hood_base"(ComponentPose=@n61, Alpha=@n62, Translation="0,0,0", Rotation="-0.000,-5,0.000", Scale="1,1,1") {
    rotationMode=BMM_Additive
    rotationSpace=BCS_BoneSpace
    alphaBoolBlend {blendOption=Linear}
    alphaScaleBiasClamp {bInterpResult=true}
    boneToModify {boneName=dyn_hood_base}
    rotation {yaw=-5}
   }
   n64 = "Transform (Modify) Bone - Bone: dyn_hood_right"(ComponentPose=@n63, Alpha=@n62, Translation="-1,0,0.000", Rotation="0,0,0", Scale="1,1,1") {
    rotationMode=BMM_Additive
    rotationSpace=BCS_BoneSpace
    translationMode=BMM_Additive
    translationSpace=BCS_BoneSpace
    alphaBoolBlend {blendOption=Linear}
    alphaScaleBiasClamp {bInterpResult=true}
    boneToModify {boneName=dyn_hood_right}
    translation {x=-1}
   }
   n65 = "Transform (Modify) Bone - Bone: lowerarm_armor_r"(ComponentPose=@n64, Alpha=Get lowerarmRAlpha(), Translation="0,0,0", Rotation="0,0.000,-45.000", Scale="1,1,1") {
    rotationMode=BMM_Additive
    rotationSpace=BCS_BoneSpace
    translationMode=BMM_Additive
    translationSpace=BCS_BoneSpace
    alphaBoolBlend {blendOption=Linear}
    boneToModify {boneName=lowerarm_armor_r}
    rotation {roll=-45}
   }
   n66 = "Constraint - armor_upper_r (PP)"(ComponentPose=@n65, Alpha=1, ConstraintWeights_0=0.65, ConstraintWeights_1=0.35) {
    alphaBoolBlend {blendOption=Linear}
    boneToModify {boneName=armor_upper_r}
    constraintSetup [{targetBone {boneName=upperarm_r}}, {targetBone {boneName=deform_upperarm_up_r}}]
    constraintWeights [1,1]
   }
   n67 = "Transform (Modify) Bone - Bone: armor_upper_r"!(ComponentPose=@n66, bAlphaBoolEnabled:bool=Get frontend_not_emote(), Translation="-4,0,-2", Rotation="0,-10,0", Scale="1,1,1") -> (Pose) {
    alphaInputType=Bool
    rotationMode=BMM_Additive
    rotationSpace=BCS_BoneSpace
    translationMode=BMM_Additive
    translationSpace=BCS_BoneSpace
    alphaBoolBlend {blendOption=Linear}
    boneToModify {boneName=armor_upper_r}
    rotation {yaw=-10}
    translation {x=-4, z=-2}
   }
   n68 = "Transform (Modify) Bone - Bone: zipper_01_r"!(ComponentPose=@n67, bAlphaBoolEnabled:bool=True, Translation="0,0,0", Rotation="-10,0,0", Scale="1,1,1") -> (Pose) {
    alphaInputType=Bool
    rotationMode=BMM_Additive
    rotationSpace=BCS_BoneSpace
    translationMode=BMM_Additive
    translationSpace=BCS_BoneSpace
    alphaBoolBlend {blendOption=Linear}
    boneToModify {boneName=zipper_01_r}
    rotation {pitch=-10}
   }
   n69 = "Transform (Modify) Bone - Bone: zipper_01_l"!(ComponentPose=@n68, bAlphaBoolEnabled:bool=True, Translation="0,0,0", Rotation="10,0,-0.000", Scale="1,1,1") -> (Pose) {
    alphaInputType=Bool
    rotationMode=BMM_Additive
    rotationSpace=BCS_BoneSpace
    translationMode=BMM_Additive
    translationSpace=BCS_BoneSpace
    alphaBoolBlend {blendOption=Linear}
    boneToModify {boneName=zipper_01_l}
    rotation {pitch=10}
   }
   Save cached pose 'RBAN Disabled Cache'(Pose=Component To Local(ComponentPose=@n69)) {cacheName=RBAN Disabled Cache}
   Use cached pose 'Stage 1 Cache'()
   Use cached pose 'Stage 2 Cache'()
  ]}
 EventGraph {nodes [
   Event BlueprintUpdateAnimation()
   > n1 = Sequence!(execute) -> (then_0, then_1, then_2:Exec, then_3:Exec, then_4:Exec)
   > Set jump_falling(jump_falling:bool=OR Boolean(A=Get bIsJumping(), B=Get bIsFalling())) -> (Output_Get:bool)
   n2 = Get bIsMoving2D()
   > Set crouch_not_moving(crouch_not_moving:bool=AND Boolean(A=NOT Boolean(A=@n2), B=Get bIsCrouching())) -> (Output_Get:bool)
   n3 = Get bIsPlayingEmote()
   > Set frontend_not_emote(frontend_not_emote:bool=AND Boolean(A=Get bIsInFrontEnd(), B=NOT Boolean(A=@n3))) -> (Output_Get:bool)
   n4 = BreakTransform(InTransform=GetDeltaTransformFromRefPose(self=GetOwnerSkeletalMeshComponent(), BoneName=head, BaseName=spine_05))
   Set head(execute=@n1.then_1, head:double=MapRangeClamped(Value=BreakRotator(InRot=@n4.Rotation).Yaw, InRangeA=0, InRangeB=-20)) -> (Output_Get:double)
   > Set neckAlpha(neckAlpha:double=MapRangeClamped(Value=0, InRangeA=0, InRangeB=30)) -> (Output_Get:double)
   n5 = BreakTransform(InTransform=GetDeltaTransformFromRefPose(self=GetOwnerSkeletalMeshComponent(), BoneName=upperarm_l, BaseName=clavicle_l))
   > Set armUpLeftAlpha(armUpLeftAlpha:double=MapRangeClamped(Value=BreakRotator(InRot=@n5.Rotation).Pitch, InRangeA=0, InRangeB=30)) -> (Output_Get:double)
   n6 = GetDeltaTransformFromRefPose(self=GetOwnerSkeletalMeshComponent(), BoneName=thigh_l, BaseName=pelvis)
   n7 = MapRangeClamped(Value=BreakRotator(InRot=BreakTransform(InTransform=@n6).Rotation).Yaw, InRangeA=-20, InRangeB=-90)
   > Set thighLAlpha(thighLAlpha:double=@n7) -> (Output_Get:double)
   n8 = GetDeltaTransformFromRefPose(self=GetOwnerSkeletalMeshComponent(), BoneName=thigh_r, BaseName=pelvis)
   n9 = MapRangeClamped(Value=BreakRotator(InRot=BreakTransform(InTransform=@n8).Rotation).Yaw, InRangeA=0, InRangeB=-90)
   > Set thighRAlpha(thighRAlpha:double=@n9) -> (Output_Get:double)
   n10 = BreakTransform(InTransform=GetDeltaTransformFromRefPose(self=GetOwnerSkeletalMeshComponent(), BoneName=lowerarm_r, BaseName=upperarm_r))
   > Set lowerarmRAlpha(lowerarmRAlpha:double=MapRangeClamped(Value=BreakRotator(InRot=@n10.Rotation).Yaw, InRangeA=0, InRangeB=60)) -> (Output_Get:double)
   > n11 = Branch(Condition=AND Boolean(A=Get bIsCrouching(), B=Get bIsMoving2D()))
   > Set thighLCrouchAlpha(thighLCrouchAlpha:double=@n7) -> (Output_Get:double)
   > Set thighRCrouchAlpha(thighRCrouchAlpha:double=@n9) -> (Output_Get:double)
   Set thighLCrouchAlpha(execute=@n11.else, thighLCrouchAlpha=0)
   > Set thighRCrouchAlpha(thighRCrouchAlpha=0)
   n12 = Branch(execute=@n1.then_2, Condition=Get jump_falling() -> (jump_falling:bool))
   n13 = BreakTransform(InTransform=GetDeltaTransformFromRefPose(self=GetOwnerSkeletalMeshComponent(), BoneName=thigh_r, BaseName=pelvis))
   Set jump_alpha(execute=[@n12.then, @n15.then], jump_alpha:double=MapRangeClamped(Value=BreakRotator(InRot=@n13.Rotation).Yaw, InRangeA=0, InRangeB=-100)) -> (Output_Get:double)
   Set jump_alpha(execute=@n12.else, jump_alpha=0)
   Set stage2(execute=@n1.then_3, stage2:bool=CharacterItemDefinitionHasAnyMetaTag(Tags=Make Array("[0]"=(TagName="Cosmetics.Variant.Property.Stage2")))) -> (Output_Get:bool)
   > Set stage3(stage3:bool=CharacterItemDefinitionHasAnyMetaTag(Tags=Make Array!("[0]"=(TagName="Cosmetics.Variant.Property.Stage3"), "[1]":Gameplay Tag Structure=(TagName="Cosmetics.Variant.Property.Stage4")) -> (Array))) -> (Output_Get:bool)
   n14 = Gravity Override State Function DEPRECATED(
    execute=@n1.then_4
    CharacterPartAnimInstance=Self-Reference()
    GameTimeInSeconds=GetGameTimeInSeconds()
    GravityOverrideParams=Get Gravity Override Params()
   )
   > Set GravityOverride(GravityOverride=@n14)
   n15 = Set thighRAlpha(thighRAlpha=0)
   Event OnCharacterCustomizationLoaded()
  ]}
]
variables [
 armUpLeftAlpha {type=number}
 crouch_not_moving {type=boolean}
 frontend_not_emote {type=boolean}
 Gravity Override Params {type=GravityOverrideParamsStruct}
 GravityOverride {type=Vector}
 head {type=number}
 jump_alpha {type=number}
 jump_falling {type=boolean}
 lowerarmRAlpha {type=number}
 neckAlpha {type=number}
 stage2 {type=boolean}
 stage3 {type=boolean}
 thighLAlpha {type=number}
 thighLCrouchAlpha {type=number}
 thighRAlpha {type=number}
 thighRCrouchAlpha {type=number}
]
